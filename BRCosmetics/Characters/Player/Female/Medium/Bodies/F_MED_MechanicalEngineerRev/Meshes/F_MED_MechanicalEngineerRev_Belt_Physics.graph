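# Physics asset for the belt mesh of the female medium MechanicalEngineerRev player character, for physics-driven belt motion.
# Five capsule bodies, dyn_belt_01 to 05, form a chain anchored to the pelvis, with a wider swing limit at the root, tighter ones between links and zero twist limits.
# Pelvis capsule and box plus a right-thigh capsule model the adjacent body, with the thigh's angular constraint left free.
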PhysicsAsset /BRCosmetics/Characters/Player/Female/Medium/Bodies/F_MED_MechanicalEngineerRev/Meshes/F_MED_MechanicalEngineerRev_Belt_Physics
bodies {
 dyn_belt_01 [{shapeType=Capsule, length=1, radius=1, rotation=<rot 85 0.0015 0.0018>}]
 dyn_belt_02 [{shapeType=Capsule, length=3, radius=2.5, center=<3.2304 0 0.3889>, rotation=<rot 85.0001 -0.0002 0.0002>}]
 dyn_belt_03 [{shapeType=Capsule, length=3, radius=2.5, center=<4.8316 -0.169 0.6685>, rotation=<rot 90 90 90>}]
 dyn_belt_04 [{shapeType=Capsule, length=3, radius=2.5, center=<5.0063 0 0.4475>, rotation=<rot 85.0002 0.0001 -0.0001>}]
 dyn_belt_05 [{shapeType=Capsule, length=2.5, radius=3.5, center=<3.248 0 0.5891>, rotation=<rot 85.0003 0.0002 89.9999>}]
 pelvis [
  {shapeType=Capsule, length=8.75, radius=9.25, center=<8.156 -1.3391 0>}
  {shapeType=Box, extentX=63.5, extentY=53, extentZ=18, shapeName=pelvis_box, center=<-5.1394 0 -9.7574>}
 ]
 thigh_r [{shapeType=Capsule, length=28, radius=11.25, center=<13.9486 1.6599 1.4893>, rotation=<rot -90 0.0001 -0.0005>}]
}
constraints [
 {bone1Name=dyn_belt_03, bone2Name=dyn_belt_02, swing1LimitDegrees=20, swing2LimitDegrees=35, twistLimitDegrees=0}
 {bone1Name=dyn_belt_04, bone2Name=dyn_belt_03, swing1LimitDegrees=20, swing2LimitDegrees=35, twistLimitDegrees=0}
 {bone1Name=dyn_belt_05, bone2Name=dyn_belt_04, swing1LimitDegrees=20, swing2LimitDegrees=35, twistLimitDegrees=0}
 {bone1Name=dyn_belt_02, bone2Name=dyn_belt_01, swing1LimitDegrees=20, swing2LimitDegrees=35, twistLimitDegrees=0}
 {
  bone1Name=thigh_r
  bone2Name=pelvis
  swing1LimitDegrees=45
  swing1Motion=Free
  swing2LimitDegrees=45
  swing2Motion=Free
  twistLimitDegrees=45
  twistMotion=Free
 }
 {bone1Name=dyn_belt_01, bone2Name=pelvis, swing1LimitDegrees=45, swing2LimitDegrees=45, twistLimitDegrees=0}
]
