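# Animation Blueprint for the F_MED_MechanicalEngineerRev female body cosmetic, following the attached parent mesh and adding secondary motion.
# The copied pose feeds a deform Control Rig, then weighted constraints place belt, knee pad and zipper base bones between body bones.
# RigidBody simulates the belt, while sprung, angle-limited AnimDynamics swing calf straps and zipper.
# The update event stores a gravity override from a deprecated function, which no node visibly reads.
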AnimBlueprint /BRCosmetics/Characters/Player/Female/Medium/Bodies/F_MED_MechanicalEngineerRev/Meshes/F_MED_MechanicalEngineerRev_AnimBP
target_skeleton=&/F_MED_MechanicalEngineerRev_Skeleton
graphs [
 AnimGraph {nodes [
   n1 = Control Rig(Source=Copy Pose From Mesh() {bUseAttachedParent=true}) {
    alphaBoolBlend {blendOption=Linear}
    controlRigAssetReference {blueprintRigClass=&/F_MED_DeformRig_MechanicalEngineerRev_ControlRig.F_MED_DeformRig_MechanicalEngineerRev_ControlRig_C}
   }
   n2 = "Constraint - dyn_belt_base (PPPPP)"!(
    ComponentPose=Local To Component(LocalPose=@n1)
    Alpha=1
    ConstraintWeights_0=0.0833
    ConstraintWeights_1:float=0.0394
    ConstraintWeights_2:float=0.3307
    ConstraintWeights_3:float=0.014
    ConstraintWeights_4:float=0.5325
   ) -> (Pose) {
    alphaBoolBlend {blendOption=Linear}
    boneToModify {boneName=dyn_belt_base}
    constraintSetup [
     {targetBone {boneName=spine_04}}
     {targetBone {boneName=spine_03}}
     {targetBone {boneName=spine_02}}
     {targetBone {boneName=spine_01}}
     {targetBone {boneName=pelvis}}
    ]
    constraintWeights [1,1,1,1,1]
   }
   n3 = RigidBody!(ComponentPose=@n2, Alpha, OverrideWorldGravity:Vector="0,0,0", ExternalForce) -> (Pose) {
    bClampLinearTranslationLimitToRefPose=true
    bOverrideWorldGravity=true
    overridePhysicsAsset=&/F_MED_MechanicalEngineerRev_Belt_Physics
    simulationSpace=BaseBoneSpace
    alphaBoolBlend {blendOption=Linear}
    baseBoneRef {boneName=spine_01}
    componentAppliedLinearAccClamp=<3000 3000 3000>
    componentLinearAccScale=<0.1 0.1 0.1>
   }
   n4 = "Constraint - dyn_kneePad_l (PP)"!(ComponentPose=@n3, Alpha=1, ConstraintWeights_0=0.5, ConstraintWeights_1:float=0.5) -> (Pose) {
    alphaBoolBlend {blendOption=Linear}
    boneToModify {boneName=dyn_kneePad_l}
    constraintSetup [{targetBone {boneName=thigh_l}}, {targetBone {boneName=calf_l}}]
    constraintWeights [1,1]
   }
   n5 = "Constraint - dyn_kneePad_r (PP)"!(ComponentPose=@n4, Alpha=1, ConstraintWeights_0=0.5, ConstraintWeights_1:float=0.5) -> (Pose) {
    alphaBoolBlend {blendOption=Linear}
    boneToModify {boneName=dyn_kneePad_r}
    constraintSetup [{targetBone {boneName=thigh_r}}, {targetBone {boneName=calf_r}}]
    constraintWeights [1,1]
   }
   n6 = Anim Dynamics - Bone: dyn_calfStrap_l!(ComponentPose=@n5, Alpha=1, GravityOverride:Vector="0,0,0", ExternalForce="0,0,0") -> (Pose) {
    angularDampingOverride=0.8
    angularSpringConstant=75
    bAngularSpring=true
    bGravityOverrideInSimSpace=true
    bUseGravityOverride=true
    linearDampingOverride=0.9
    linearSpringConstant=150
    numSolverIterationsPreUpdate=3
    alphaBoolBlend {blendOption=Linear}
    boundBone {boneName=dyn_calfStrap_l}
    componentAppliedLinearAccClamp=<10000 10000 10000>
    componentLinearVelScale=<0.1 0.1 0.1>
    gravityOverride {z=-350}
    physicsBodyDefinitions [{
      boundBone {boneName=dyn_calfStrap_l}
      boxExtents=<8 8 8>
      constraintSetup {angularLimitsMax=<0 0 10>, angularLimitsMin=<0 0 -10>, angularTarget=<1 0 0>}
      localJointOffset=<2 0 0>
     }]
    relativeSpaceBone {boneName=spine_05}
   }
   n7 = Anim Dynamics - Bone: dyn_calfStrap_r!(ComponentPose=@n6, Alpha=1, GravityOverride:Vector="0,0,0", ExternalForce="0,0,0") -> (Pose) {
    angularDampingOverride=0.8
    angularSpringConstant=75
    bAngularSpring=true
    bGravityOverrideInSimSpace=true
    bUseGravityOverride=true
    linearDampingOverride=0.9
    linearSpringConstant=150
    numSolverIterationsPreUpdate=3
    alphaBoolBlend {blendOption=Linear}
    boundBone {boneName=dyn_calfStrap_r}
    componentAppliedLinearAccClamp=<10000 10000 10000>
    componentLinearVelScale=<0.1 0.1 0.1>
    gravityOverride {z=-350}
    physicsBodyDefinitions [{
      boundBone {boneName=dyn_calfStrap_r}
      boxExtents=<8 8 8>
      constraintSetup {angularLimitsMax=<0 0 10>, angularLimitsMin=<0 0 -10>, angularTarget=<1 0 0>}
      localJointOffset=<2 0 0>
     }]
    relativeSpaceBone {boneName=spine_05}
   }
   n8 = "Constraint - dyn_zipper_base (PPPPPP)"!(
    ComponentPose=@n7
    Alpha=1
    ConstraintWeights_0=0.0035
    ConstraintWeights_1:float=0.9166
    ConstraintWeights_2:float=0.0024
    ConstraintWeights_3:float=0.0024
    ConstraintWeights_4:float=0.072
    ConstraintWeights_5:float=0.003
   ) -> (Pose) {
    alphaBoolBlend {blendOption=Linear}
    boneToModify {boneName=dyn_zipper_base}
    constraintSetup [
     {targetBone {boneName=neck_01}}
     {targetBone {boneName=spine_05}}
     {targetBone {boneName=clavicle_l}}
     {targetBone {boneName=clavicle_r}}
     {targetBone {boneName=spine_04}}
     {targetBone {boneName=spine_03}}
    ]
    constraintWeights [1,1,1,1,1,1]
   }
   n9 = Anim Dynamics - Bone: dyn_zipper!(ComponentPose=@n8, Alpha=1, GravityOverride:Vector="0,0,-350", ExternalForce) -> (Pose) {
    angularDampingOverride=0.8
    angularSpringConstant=75
    bAngularSpring=true
    bGravityOverrideInSimSpace=true
    bUseGravityOverride=true
    linearDampingOverride=0.9
    linearSpringConstant=150
    numSolverIterationsPreUpdate=3
    alphaBoolBlend {blendOption=Linear}
    boundBone {boneName=dyn_zipper}
    componentAppliedLinearAccClamp=<10000 10000 10000>
    componentLinearAccScale=<0.05 0.05 0.2>
    componentLinearVelScale=<0.1 0.1 0.1>
    gravityOverride {z=-350}
    physicsBodyDefinitions [{
      boundBone {boneName=dyn_zipper}
      boxExtents=<8 8 8>
      constraintSetup {angularLimitsMax=<0 5 1>, angularLimitsMin=<0 -30 -1>, angularTarget=<1 0 0>}
      localJointOffset=<2 0 0>
     }]
    relativeSpaceBone {boneName=spine_05}
   }
   Output Pose(Result=Component To Local(ComponentPose=@n9)) {layerGroup=None}
  ]}
 EventGraph {nodes [
   Event BlueprintUpdateAnimation()
   > n1 = Gravity Override State Function DEPRECATED(
    CharacterPartAnimInstance=Self-Reference()
    GameTimeInSeconds=GetGameTimeInSeconds()
    GravityOverrideParams=Get gravity override params() -> (gravity override params:Gravity Override Params Struct Structure)
   )
   > Set gravity override(gravity override:Vector=@n1) -> (Output_Get:Vector)
  ]}
]
variables [
 gravity override {type=Vector}
 gravity override params {type=GravityOverrideParamsStruct}
]
