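# Physics asset for the MechanicalEngineerRev female face accessory mesh, defining hair physics for two braids, a cowlick and two side bangs.
# Each hair strand is a capsule chain constrained back to the head, with little or no twist and braid swing limits that widen toward the tips.
# Head, neck, upper spine, clavicle and upper arm bodies supply body volumes; only the left upper arm has a constraint, fully free in rotation.
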PhysicsAsset /BRCosmetics/Characters/Player/Female/Medium/Bodies/F_MED_MechanicalEngineerRev/Meshes/Parts/F_MED_MechanicalEngineerRev_FaceAcc_Physics
bodies {
 clavicle_l [{shapeType=Capsule, length=7.625, radius=4.875, center=<3.4122 -1.7025 -0.6227>, rotation=<rot 82.9999 0 0>}]
 clavicle_r [{shapeType=Capsule, length=10.5, radius=4.75, center=<-4.5408 0 0.4821>, rotation=<rot 82.9999 0 0>}]
 dyn_braid_l_01 [{shapeType=Capsule, length=3.25, radius=2, center=<2.7902 0 0>, rotation=<rot 90 -179.9998 180>}]
 dyn_braid_l_02 [{shapeType=Capsule, length=3, radius=1.75, center=<2.7902 0 0>, rotation=<rot 90 -179.8416 -179.8418>}]
 dyn_braid_l_03 [{shapeType=Capsule, length=3.75, radius=1.5, center=<3.8389 0 0>, rotation=<rot 90 -160.5288 -160.529>}]
 dyn_braid_l_04 [{shapeType=Capsule, length=3.75, radius=1.5, center=<2.7902 0 0>, rotation=<rot 90 180 179.9998>}]
 dyn_braid_l_05 [{shapeType=Capsule, length=4.5, radius=1.5, center=<2.7902 0 0>, rotation=<rot 90 -179.9802 -179.9803>}]
 dyn_braid_r_01 [{shapeType=Capsule, length=4.5, radius=1.5, center=<2.7902 0 0>, rotation=<rot 90 0 0.0002>}]
 dyn_braid_r_02 [{shapeType=Capsule, length=4.5, radius=1.5, center=<2.7902 0 0>, rotation=<rot 90 6.7214 6.7215>}]
 dyn_braid_r_03 [{shapeType=Capsule, length=4.5, radius=1.5, center=<2.7902 0 0>, rotation=<rot 90 166.7373 166.7375>}]
 dyn_braid_r_04 [{shapeType=Capsule, length=4.5, radius=1.5, center=<2.7902 0 0>, rotation=<rot 90 179.9802 179.9803>}]
 dyn_braid_r_05 [{shapeType=Capsule, length=4.5, radius=1.5, center=<2.7902 0 0>, rotation=<rot 90 90 90.0001>}]
 dyn_cowlick_01 [{shapeType=Capsule, length=1, radius=1, center=<1.0522 0 0>, rotation=<rot -90 -179.995 179.9954>}]
 dyn_cowlick_02 [{shapeType=Capsule, length=1.5, radius=1, center=<1.8294 0 0>, rotation=<rot -90 -96.113 106.1136>}]
 dyn_cowlick_03 [{shapeType=Capsule, length=1, radius=1, center=<0.639 -0.1583 -0.1128>, rotation=<rot -90 -6.7214 6.7218>}]
 head [{shapeType=Capsule, length=7.75, radius=7.5, center=<3 1.1004 0>, rotation=<rot 90 0 0>}]
 neck_02 [{shapeType=Box, extentX=35, extentY=24.5, extentZ=56.51, center=<-4.3908 -13.2529 -0.7955>, rotation=<rot 0 -12.2778 0>}]
 sideBang_l_01 [{shapeType=Capsule, length=1.5, radius=1.5, center=<1.0577 0 0>, rotation=<rot 90 -160.5289 -160.5293>}]
 sideBang_l_02 [{shapeType=Capsule, length=1.5, radius=1.5, center=<1.0577 0 0>, rotation=<rot 90 -0.0001 -0.0003>}]
 sideBang_l_03 [{shapeType=Capsule, length=1.5, radius=1.5, center=<1.0577 0 0>, rotation=<rot 90 -90 -90.0002>}]
 sideBang_r_01 [{shapeType=Capsule, length=1.5, radius=1.5, center=<1.0577 0 0>, rotation=<rot 90 179.9801 179.9805>}]
 sideBang_r_02 [{shapeType=Capsule, length=1.5, radius=1.5, center=<1.0577 0 0>, rotation=<rot 90 179.9803 179.9806>}]
 sideBang_r_03 [{shapeType=Capsule, length=1.5, radius=1.5, center=<1.0577 0 0>, rotation=<rot 90 174.9488 174.949>}]
 spine_05 [{shapeType=Capsule, length=13, radius=9, center=<-4.7342 3.5607 0>, rotation=<rot 0 -90 0>}]
 upperarm_l [{
   shapeType=Box
   extentX=24.5
   extentY=7.5
   extentZ=23.5
   shapeName=upperarm_r_box
   center=<5.6547 4.8737 5.7301>
   rotation=<rot 0 0.0002 -120.0002>
  }]
 upperarm_r [{
   shapeType=Box
   extentX=20
   extentY=7.5
   extentZ=19
   shapeName=upperarm_r_box
   center=<-5.6547 -2.7334 -3.6318>
   rotation=<rot 0 0.0001 44.9999>
  }]
}
constraints [
 {bone1Name=dyn_cowlick_03, bone2Name=dyn_cowlick_02, swing1LimitDegrees=7, swing2LimitDegrees=10, twistLimitDegrees=0}
 {bone1Name=dyn_cowlick_02, bone2Name=dyn_cowlick_01, swing1LimitDegrees=5, swing2LimitDegrees=5, twistLimitDegrees=0}
 {bone1Name=dyn_cowlick_01, bone2Name=head, swing1LimitDegrees=1, swing2LimitDegrees=1, twistLimitDegrees=0}
 {bone1Name=dyn_braid_l_01, bone2Name=head, swing1LimitDegrees=1, swing2LimitDegrees=30, twistLimitDegrees=2}
 {bone1Name=dyn_braid_l_02, bone2Name=dyn_braid_l_01, swing1LimitDegrees=1, swing2LimitDegrees=30, twistLimitDegrees=2}
 {bone1Name=dyn_braid_l_03, bone2Name=dyn_braid_l_02, swing1LimitDegrees=5, swing2LimitDegrees=45, twistLimitDegrees=2}
 {bone1Name=dyn_braid_l_04, bone2Name=dyn_braid_l_03, swing1LimitDegrees=20, swing2LimitDegrees=45, twistLimitDegrees=2}
 {bone1Name=dyn_braid_l_05, bone2Name=dyn_braid_l_04, swing1LimitDegrees=20, swing2LimitDegrees=45, twistLimitDegrees=2}
 {bone1Name=dyn_braid_r_01, bone2Name=head, swing1LimitDegrees=1, swing2LimitDegrees=30, twistLimitDegrees=2}
 {bone1Name=dyn_braid_r_02, bone2Name=dyn_braid_r_01, swing1LimitDegrees=1, swing2LimitDegrees=30, twistLimitDegrees=2}
 {bone1Name=dyn_braid_r_03, bone2Name=dyn_braid_r_02, swing1LimitDegrees=10, swing2LimitDegrees=45, twistLimitDegrees=2}
 {bone1Name=dyn_braid_r_04, bone2Name=dyn_braid_r_03, swing1LimitDegrees=20, swing2LimitDegrees=45, twistLimitDegrees=2}
 {bone1Name=dyn_braid_r_05, bone2Name=dyn_braid_r_04, swing1LimitDegrees=20, swing2LimitDegrees=45, twistLimitDegrees=2}
 {bone1Name=sideBang_l_01, bone2Name=head, swing1LimitDegrees=20, swing2LimitDegrees=5, twistLimitDegrees=0}
 {bone1Name=sideBang_l_02, bone2Name=sideBang_l_01, swing1LimitDegrees=20, swing2LimitDegrees=5, twistLimitDegrees=0}
 {bone1Name=sideBang_l_03, bone2Name=sideBang_l_02, swing1LimitDegrees=20, swing2LimitDegrees=5, twistLimitDegrees=0}
 {bone1Name=sideBang_r_01, bone2Name=head, swing1LimitDegrees=5, swing2LimitDegrees=5, twistLimitDegrees=0}
 {bone1Name=sideBang_r_02, bone2Name=sideBang_r_01, swing1LimitDegrees=5, swing2LimitDegrees=5, twistLimitDegrees=0}
 {bone1Name=sideBang_r_03, bone2Name=sideBang_r_02, swing1LimitDegrees=5, swing2LimitDegrees=5, twistLimitDegrees=0}
 {
  bone1Name=upperarm_l
  bone2Name=clavicle_l
  swing1LimitDegrees=45
  swing1Motion=Free
  swing2LimitDegrees=45
  swing2Motion=Free
  twistLimitDegrees=45
  twistMotion=Free
 }
]
