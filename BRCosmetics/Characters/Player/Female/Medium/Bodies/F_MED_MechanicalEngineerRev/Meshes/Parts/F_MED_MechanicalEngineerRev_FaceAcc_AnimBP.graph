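# Post-process style AnimBP for the female MechanicalEngineerRev face accessory, adding braid dynamics to the pose copied from the attached parent mesh.
# Two neck-based RigidBody sims (standard and FrontEnd physics assets) feed clavicle constraints and additive braid bone offsets; spine_05 constraints sit unconnected.
# The update event sets gravity override through a deprecated helper, head-up amount, and skydive, physics and frontend alphas.
# The frontend check always overwrites the skydive half-strength physics alpha.
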AnimBlueprint /BRCosmetics/Characters/Player/Female/Medium/Bodies/F_MED_MechanicalEngineerRev/Meshes/Parts/F_MED_MechanicalEngineerRev_FaceAcc_AnimBP
target_skeleton=&/F_MED_MechanicalEngineerRev_FaceAcc_Skeleton
graphs [
 AnimGraph {nodes [
   n1 = "Constraint - dyn_braid_l_03 (P)"(Alpha=1, ConstraintWeights_0=1) {
    alphaBoolBlend {blendOption=Linear}
    boneToModify {boneName=dyn_braid_l_03}
    constraintSetup [{targetBone {boneName=spine_05}}]
    constraintWeights [1]
   }
   "Constraint - dyn_braid_r_03 (P)"(ComponentPose=@n1, Alpha=1, ConstraintWeights_0=1) {
    alphaBoolBlend {blendOption=Linear}
    boneToModify {boneName=dyn_braid_r_03}
    constraintSetup [{targetBone {boneName=spine_05}}]
    constraintWeights [1]
   }
   n2 = RigidBody!(
    ComponentPose=Local To Component(LocalPose=Copy Pose From Mesh() {bUseAttachedParent=true})
    Alpha
    OverrideWorldGravity:Vector="0,0,0"
    ExternalForce
   ) -> (Pose) {
    bClampLinearTranslationLimitToRefPose=true
    bOverrideWorldGravity=true
    overridePhysicsAsset=&/F_MED_MechanicalEngineerRev_FaceAcc_Physics
    simulationSpace=BaseBoneSpace
    alphaBoolBlend {blendOption=Linear}
    baseBoneRef {boneName=neck_01}
    componentAppliedLinearAccClamp=<3000 3000 3000>
    componentLinearAccScale=<0.4 0.4 0.5>
    componentLinearVelScale=<0.1 0.1 0.1>
   }
   n3 = RigidBody(ComponentPose=@n2, OverrideWorldGravity="0,0,0") {
    bClampLinearTranslationLimitToRefPose=true
    bOverrideWorldGravity=true
    overridePhysicsAsset=&/F_MED_MechanicalEngineerRev_FaceAcc_FrontEnd_Physics
    simulationSpace=BaseBoneSpace
    alphaBoolBlend {blendOption=Linear}
    baseBoneRef {boneName=neck_01}
    componentAppliedLinearAccClamp=<3000 3000 3000>
    componentLinearAccScale=<0.2 0.2 0.1>
    componentLinearVelScale=<0.1 0.1 0.1>
   }
   n4 = "Constraint - dyn_braid_l_01 (P)"(ComponentPose=@n3, Alpha=1, ConstraintWeights_0=1) {
    alphaBoolBlend {blendOption=Linear}
    boneToModify {boneName=dyn_braid_l_01}
    constraintSetup [{targetBone {boneName=clavicle_l}}]
    constraintWeights [1]
   }
   n5 = "Constraint - dyn_braid_l_02 (P)"(ComponentPose=@n4, Alpha=1, ConstraintWeights_0=1) {
    alphaBoolBlend {blendOption=Linear}
    boneToModify {boneName=dyn_braid_l_02}
    constraintSetup [{targetBone {boneName=clavicle_l}}]
    constraintWeights [1]
   }
   n6 = "Constraint - dyn_braid_r_01 (P)"(ComponentPose=@n5, Alpha=1, ConstraintWeights_0=1) {
    alphaBoolBlend {blendOption=Linear}
    boneToModify {boneName=dyn_braid_r_01}
    constraintSetup [{targetBone {boneName=clavicle_r}}]
    constraintWeights [1]
   }
   n7 = "Constraint - dyn_braid_r_02 (P)"(ComponentPose=@n6, Alpha=1, ConstraintWeights_0=1) {
    alphaBoolBlend {blendOption=Linear}
    boneToModify {boneName=dyn_braid_r_02}
    constraintSetup [{targetBone {boneName=clavicle_r}}]
    constraintWeights [1]
   }
   n8 = "Transform (Modify) Bone - Bone: dyn_braid_l_03"(ComponentPose=@n7, Alpha=1, Translation="0,0,0", Rotation="-0.000,-30.000,0.000", Scale="1,1,1") {
    rotationMode=BMM_Additive
    translationMode=BMM_Additive
    alphaBoolBlend {blendOption=Linear}
    alphaScaleBiasClamp {bInterpResult=true, interpSpeedDecreasing=5, interpSpeedIncreasing=5}
    boneToModify {boneName=dyn_braid_l_03}
    rotation {yaw=-30}
   }
   n9 = "Transform (Modify) Bone - Bone: dyn_braid_r_03"(ComponentPose=@n8, Alpha=1, Translation="0,0,0", Rotation="0,30,0", Scale="1,1,1") {
    rotationMode=BMM_Additive
    translationMode=BMM_Additive
    alphaBoolBlend {blendOption=Linear}
    alphaScaleBiasClamp {bInterpResult=true, interpSpeedDecreasing=5, interpSpeedIncreasing=5}
    boneToModify {boneName=dyn_braid_r_03}
    rotation {yaw=30}
   }
   n10 = "Transform (Modify) Bone - Bone: dyn_braid_l_04"(ComponentPose=@n9, Alpha=1, Translation="0,0,0", Rotation="0.000,10.000,-0.000", Scale="1,1,1") {
    rotationMode=BMM_Additive
    translationMode=BMM_Additive
    alphaBoolBlend {blendOption=Linear}
    alphaScaleBiasClamp {bInterpResult=true, interpSpeedDecreasing=5, interpSpeedIncreasing=5}
    boneToModify {boneName=dyn_braid_l_04}
    rotation {yaw=10}
   }
   n11 = "Transform (Modify) Bone - Bone: dyn_braid_r_04"(ComponentPose=@n10, Alpha=1, Translation="0,0,0", Rotation="0,-10,0", Scale="1,1,1") {
    rotationMode=BMM_Additive
    translationMode=BMM_Additive
    alphaBoolBlend {blendOption=Linear}
    alphaScaleBiasClamp {bInterpResult=true, interpSpeedDecreasing=5, interpSpeedIncreasing=5}
    boneToModify {boneName=dyn_braid_r_04}
    rotation {yaw=-10}
   }
   n12 = "Transform (Modify) Bone - Bone: dyn_braid_l_02"(ComponentPose=@n11, Alpha=1, Translation="-3,0,0", Rotation="0,0,0", Scale="1,1,1") {
    rotationMode=BMM_Additive
    translationMode=BMM_Additive
    alphaBoolBlend {blendOption=Linear}
    alphaScaleBiasClamp {bInterpResult=true, interpSpeedDecreasing=5, interpSpeedIncreasing=5}
    boneToModify {boneName=dyn_braid_l_02}
    translation {x=-3}
   }
   n13 = "Transform (Modify) Bone - Bone: dyn_braid_r_02"(ComponentPose=@n12, Alpha=1, Translation="1,0,0", Rotation="0,0,0", Scale="1,1,1") {
    rotationMode=BMM_Additive
    translationMode=BMM_Additive
    alphaBoolBlend {blendOption=Linear}
    alphaScaleBiasClamp {bInterpResult=true, interpSpeedDecreasing=5, interpSpeedIncreasing=5}
    boneToModify {boneName=dyn_braid_r_02}
    translation {x=1}
   }
   Output Pose(Result=Component To Local(ComponentPose=@n13)) {layerGroup=None}
  ]}
 EventGraph {nodes [
   Event BlueprintUpdateAnimation()
   > n1 = Gravity Override State Function DEPRECATED(CharacterPartAnimInstance=Self-Reference(), GameTimeInSeconds=GetGameTimeInSeconds(), GravityOverrideParams=Get Gravity Override Params())
   > Set Gravity Override(Gravity Override=@n1)
   n2 = BreakTransform(InTransform=GetDeltaTransformFromRefPose(self=GetOwnerSkeletalMeshComponent(), BoneName=head, BaseName=spine_05))
   > Set Head Up(Head Up:double=MapRangeClamped(Value=BreakRotator(InRot=@n2.Rotation).Yaw, InRangeA=-20, InRangeB=-41)) -> (Output_Get:double)
   > n3 = Branch(Condition=OR Boolean(A=Get bIsSkydiveFloating(), B=Get bIsSkydiveDiveMode()))
   > Set RBAN Alpha(RBAN Alpha:double=0.5) -> (Output_Get:double)
   > n4 = Set Skydive Alpha(Skydive Alpha:double=1) -> (Output_Get:double)
   n5 = Branch(execute=[@n6.then, @n4.then], Condition=IsCosmeticObjectDisplayedInFrontendContext!(CosmeticObject:Object Reference) -> (ReturnValue:bool))
   > Set RBAN Alpha(RBAN Alpha=0)
   > Set FE Alpha(FE Alpha:double=1) -> (Output_Get:double)
   Set RBAN Alpha(execute=@n5.else, RBAN Alpha=1)
   > Set FE Alpha(FE Alpha=0)
   Set RBAN Alpha(execute=@n3.else, RBAN Alpha=1)
   > n6 = Set Skydive Alpha(Skydive Alpha=0)
  ]}
]
variables [
 FE Alpha {type=number}
 Gravity Override {type=Vector}
 Gravity Override Params {type=GravityOverrideParamsStruct}
 Head Up {type=number}
 RBAN Alpha {type=number}
 Skydive Alpha {type=number}
]
